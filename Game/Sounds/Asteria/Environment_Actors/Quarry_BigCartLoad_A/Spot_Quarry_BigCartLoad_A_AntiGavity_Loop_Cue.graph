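# Looping ambient Sound Cue for the anti-gravity sound of the Quarry BigCartLoad_A environment actor in Asteria.
# A Mixer layers Very Close, Close and Distant looping waves, each through its own attenuation asset to vary the sound by distance, with the Distant input boosted.
# It plays in the Athena ambient sound class and overrides concurrency with a small per-owner voice limit.
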SoundCue /Game/Sounds/Asteria/Environment_Actors/Quarry_BigCartLoad_A/Spot_Quarry_BigCartLoad_A_AntiGavity_Loop_Cue
root=Mixer_1
nodes [
 {
  id=Attenuation_0
  children [WavePlayer_3]
  properties {attenuationSettings=&/Spot_BigCartLoad_A_AntiGavity_Loop_Close_Att}
 }
 {
  id=Attenuation_1
  children [WavePlayer_6]
  properties {attenuationSettings=&/Spot_BigCartLoad_A_AntiGavity_Loop_Distant_Att}
 }
 {
  id=Attenuation_2
  children [WavePlayer_0]
  properties {attenuationSettings=&/Spot_BigCartLoad_A_AntiGavity_Loop_Very_Close_Att}
 }
 {id=Mixer_1, children [Attenuation_2, Attenuation_0, Attenuation_1], properties {inputVolume [1,1,2]}}
 {
  id=WavePlayer_0
  properties {
   bLooping=true
   soundWaveAssetPtr=&/Spot_Quarry_BigCartLoad_A_AntiGavity_Loop_Very_Close
  }
 }
 {
  id=WavePlayer_3
  properties {
   bLooping=true
   soundWaveAssetPtr=&/Spot_Quarry_BigCartLoad_A_AntiGavity_Loop_Close
  }
 }
 {
  id=WavePlayer_6
  properties {
   bLooping=true
   soundWaveAssetPtr=&/Spot_Quarry_BigCartLoad_A_AntiGavity_Loop_Distant
  }
 }
]
properties {
 bOverrideConcurrency=true
 maxDistance=1900
 soundClassObject=&Fort_Ambient_Athena
 subtitlePriority=5000
 volumeMultiplier=1.002
 attenuationOverrides {
  bEnableFocusInterpolation=true
  bEnableListenerFocus=true
  bEnableOcclusion=true
  dBAttenuationAtMax=-40
  falloffDistance=512
  focusAzimuth=15
  focusVolumeAttenuation=1.5
  nonFocusAzimuth=45
  nonFocusVolumeAttenuation=0.5
  occlusionInterpolationTime=1
  occlusionLowPassFilterFrequency=10000
  occlusionTraceChannel=ECC_GameTraceChannel11
  occlusionVolumeAttenuation=0.5
  stereoSpread=512
  attenuationShapeExtents=<256 64.5942 90>
 }
 concurrencyOverrides {bLimitToOwner=true, maxCount=3}
}
